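; Example building actor in a test map that shows enum-driven construction script setup of a point light and text render.
; The construction script switches on an enum variable to set a Small, Medium or Large label and a matching red, green or blue light color.
; A separate SetStructureData function applies a struct's color, intensity, text and text size to the same components, with no caller shown here.
; BeginPlay makes the actor invulnerable and disables damage.
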(event EventBeginPlay
  ; Make Actor Invulnerable
  (Variables|BuildingActor|SetbIsInvulnerable true)
  (Variables|Actor|SetbCanBeDamaged false))

(fn SetStructureData
  (params (Data :type Scripting%20Structure%20Example%20Structure))
  (bind (_color_3_cc2bc8a0450bfcbf816869a964f1f04d _intensity_6_bd41c53d439ceb18170d4389f156498d _text_9_b889dc7d4537d59ec53305b9a016a179 _textsize_13_dd0b7b5e4ad8e98f2a7eb9b247568d11) (Utilities|Struct|BreakScriptingStructureExample Data))
  (Rendering|Components|Light|SetLightColor (Variables|Default|GetPointLight) _color_3_cc2bc8a0450bfcbf816869a964f1f04d true)
  (Rendering|Components|Light|SetIntensity (Variables|Default|GetPointLight) _intensity_6_bd41c53d439ceb18170d4389f156498d)
  (Rendering|Components|TextRender|SetText (Variables|Default|GetTextRender) _text_9_b889dc7d4537d59ec53305b9a016a179)
  (Rendering|Components|TextRender|SetWorldSize (Variables|Default|GetTextRender) (Math|Conversions|ToFloat(Integer) _textsize_13_dd0b7b5e4ad8e98f2a7eb9b247568d11)))

(fn ConstructionScript
  (switch Utilities|FlowControl|Switch|SwitchonScriptingEnumExample (Variables|Default|GetConstructionScriptEnumExample)
    (:NewEnumerator0
      (Rendering|Components|TextRender|SetText (Variables|Default|GetTextRender) (Utilities|Text|MakeLiteralText "NSLOCTEXT(\"[FE391E4AF9BCEB412477367FAE7F50DE]\", \"97DC42134B4F563006E7CE8684466C93\", \"Small Enum\")"))
      (Rendering|Components|Light|SetLightColor (Variables|Default|GetPointLight) "(R=1.000000,G=0.000000,B=0.000000,A=1.000000)" true))
    (:NewEnumerator1
      (Rendering|Components|TextRender|SetText (Variables|Default|GetTextRender) (Utilities|Text|MakeLiteralText "NSLOCTEXT(\"[FE391E4AF9BCEB412477367FAE7F50DE]\", \"FEEE1AAB413A02C128624A83B9FCBE44\", \"Medium Enum\")"))
      (Rendering|Components|Light|SetLightColor (Variables|Default|GetPointLight) "(R=0.000000,G=1.000000,B=0.000000,A=1.000000)" true))
    (:NewEnumerator2
      (Rendering|Components|TextRender|SetText (Variables|Default|GetTextRender) (Utilities|Text|MakeLiteralText "NSLOCTEXT(\"[FE391E4AF9BCEB412477367FAE7F50DE]\", \"F36ED2B44A64B0CF79A0FAA424851790\", \"Large Enum\")"))
      (Rendering|Components|Light|SetLightColor (Variables|Default|GetPointLight) "(R=0.000000,G=0.000000,B=1.000000,A=1.000000)" true))))
