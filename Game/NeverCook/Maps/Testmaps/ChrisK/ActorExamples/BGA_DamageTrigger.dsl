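; A building actor example that deals fixed environmental damage to player pawns that touch its collision box.
; On BeginPlay it makes itself invulnerable and not damageable.
; On the server, each overlap start casts the other actor to FortPlayerPawnAthena, stores it as CurrentPawn, and applies global environment damage unless the pawn is DBNO.
; Damage fires once per overlap start, with no repeat while inside.
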(event EventBeginPlay
  ; Make Actor Invulnerable
  (Variables|BuildingActor|SetbIsInvulnerable true)
  (Variables|Actor|SetbCanBeDamaged false))

(event OnComponentBeginOverlap(CollisionBox)
  (params (OverlappedComponent :type Primitive%20Component%20Object%20Reference) (OtherActor :type Actor%20Object%20Reference) (OtherComp :type Primitive%20Component%20Object%20Reference) (OtherBodyIndex :type Integer) (bFromSweep :type Boolean) (SweepResult :type Hit%20Result%20Structure%20%28by%20ref%29))
  ; Trigger Events
  (Networking|SwitchHasAuthority
    (:Authority
      (bind _asfort_player_pawn_athena (Utilities|Casting|CastToFortPlayerPawnAthena OtherActor)) ; this checks, on the server, if a player touched this, not a bot or wildlife
      (bind _output_get (Variables|Default|SetCurrentPawn _asfort_player_pawn_athena))
      (if (not (Variables|/Script/FortniteGame.FortPawn|GetbIsDBNO _output_get))
        (Environment|ApplyGlobalEnvironmentDamagetoActor _output_get 2e+01)))))

(fn ConstructionScript)
