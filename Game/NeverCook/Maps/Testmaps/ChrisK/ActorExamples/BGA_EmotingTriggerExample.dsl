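; Example building actor, invulnerable, that shows a point light while a player emotes inside its collision box.
; On the server, a player entering the box starts a looping timer that polls overlapping player pawns for an active emote.
; The result sets IsInsideTriggerAndEmoting, whose RepNotify toggles the light's visibility.
; When the last player leaves, the flag clears and the timer pauses.
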(event EventBeginPlay
  ; Make Actor Invulnerable
  (Variables|BuildingActor|SetbIsInvulnerable true)
  (Variables|Actor|SetbCanBeDamaged false))

(event OnComponentBeginOverlap(CollisionBox)
  (params (OverlappedComponent :type Primitive%20Component%20Object%20Reference) (OtherActor :type Actor%20Object%20Reference) (OtherComp :type Primitive%20Component%20Object%20Reference) (OtherBodyIndex :type Integer) (bFromSweep :type Boolean) (SweepResult :type Hit%20Result%20Structure%20%28by%20ref%29))
  ; Trigger Events
  (Networking|SwitchHasAuthority
    (:Authority
      (Utilities|Casting|CastToFortPlayerPawnAthena OtherActor) ; this checks, on the server, if a player touched this, not a bot or wildlife
      (bind _returnvalue (Utilities|Time|SetTimerbyEvent (delegate Custom|CheckEmoting) 2.0 true)) ; if the box was touched, start a looping timer to check if player is emoting
      (Variables|Default|SetTimerHandle _returnvalue)
      (bind _playerisemoting (CallFunction|IsAnyPlayerEmoting (Collision|GetOverlappingActors (Variables|Default|GetCollisionBox) "/Script/FortniteGame.FortPlayerPawnAthena" :!class /Script/Engine.PrimitiveComponent)))
      (bind _output_get (Variables|Default|SetIsInsideTriggerAndEmoting _playerisemoting))
      (if _output_get
        _
        (elif (Utilities|Time|IsTimerPausedbyHandle (Variables|Default|GetTimerHandle))
          (Utilities|Time|UnpauseTimerbyHandle (Variables|Default|GetTimerHandle)))))))

(event Custom|CheckEmoting
  ; Trigger Events
  (bind _playerisemoting (CallFunction|IsAnyPlayerEmoting (Collision|GetOverlappingActors (Variables|Default|GetCollisionBox) "/Script/FortniteGame.FortPlayerPawnAthena" :!class /Script/Engine.PrimitiveComponent)))
  (bind _output_get (Variables|Default|SetIsInsideTriggerAndEmoting _playerisemoting))
  (if _output_get
    _
    (elif (Utilities|Time|IsTimerPausedbyHandle (Variables|Default|GetTimerHandle))
      (Utilities|Time|UnpauseTimerbyHandle (Variables|Default|GetTimerHandle)))))

(event OnComponentEndOverlap(CollisionBox)
  (params (OverlappedComponent :type Primitive%20Component%20Object%20Reference) (OtherActor :type Actor%20Object%20Reference) (OtherComp :type Primitive%20Component%20Object%20Reference) (OtherBodyIndex :type Integer))
  ; Trigger Events
  (Networking|SwitchHasAuthority
    (:Authority
      (Utilities|Casting|CastToFortPlayerPawnAthena OtherActor) ; this checks, on the server, if a player touched this, not a bot or wildlife
      (if (== (Utilities|Array|Length (Collision|GetOverlappingActors (Variables|Default|GetCollisionBox) "/Script/FortniteGame.FortPlayerPawnAthena" :!class /Script/Engine.PrimitiveComponent)) 0)
        (Variables|Default|SetIsInsideTriggerAndEmoting false)
        (Utilities|Time|PauseTimerbyHandle (Variables|Default|GetTimerHandle))))))

(fn IsAnyPlayerEmoting
  (params (Players :type Array%20of%20Actor%20Object%20References))
  (bind _array_element (Utilities|Array|ForEachLoopwithBreak Players)
    (:LoopBody
      (bind _asfort_player_pawn_athena (Utilities|Casting|CastToFortPlayerPawnAthena _array_element))
      (bind _asfort_player_controller_athena (Utilities|Casting|CastToFortPlayerControllerAthena (Pawn|GetController _asfort_player_pawn_athena)))
      (if (Emote|IsPlayingEmote _asfort_player_controller_athena)
        (LocalVariables|SetIsEmotingLocal true)))
    (:Completed
      _))
  (return (LocalVariables|GetIsEmotingLocal)))

(fn OnRep_IsInsideTriggerAndEmoting
  (if (Variables|Default|GetIsInsideTriggerAndEmoting)
    (Rendering|SetVisibility (Variables|Default|GetPointLight) true false)
    (else
      (Rendering|SetVisibility (Variables|Default|GetPointLight) false false))))

(fn ConstructionScript)
