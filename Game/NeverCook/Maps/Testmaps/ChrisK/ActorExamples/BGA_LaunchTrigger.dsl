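; BGA_LaunchTrigger is an example test-map building actor, invulnerable from BeginPlay, that launches player pawns overlapping its CollisionBox.
; On authority, it casts the overlapping actor to FortPlayerPawnAthena, stores it, and overrides its velocity with a launch along the pawn's horizontal facing plus fixed upward speed.
; It also makes the pawn ignore its next falling damage.
; Remote machines instead play a bounce pad sound at the pawn's location.
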(event EventBeginPlay
  ; Make Actor Invulnerable
  (Variables|BuildingActor|SetbIsInvulnerable true)
  (Variables|Actor|SetbCanBeDamaged false))

(event OnComponentBeginOverlap(CollisionBox)
  (params (OverlappedComponent :type Primitive%20Component%20Object%20Reference) (OtherActor :type Actor%20Object%20Reference) (OtherComp :type Primitive%20Component%20Object%20Reference) (OtherBodyIndex :type Integer) (bFromSweep :type Boolean) (SweepResult :type Hit%20Result%20Structure%20%28by%20ref%29))
  ; Trigger Events
  (Networking|SwitchHasAuthority
    (:Authority
      (bind _asfort_player_pawn_athena (Utilities|Casting|CastToFortPlayerPawnAthena OtherActor)) ; this checks, on the server, if a player touched this, not a bot or wildlife
      (Variables|Default|SetCurrentPawn _asfort_player_pawn_athena)
      (Character|LaunchCharacter (Variables|Default|GetCurrentPawn) (Math|Vector|MakeVector (.x (* (Transformation|GetActorForwardVector (Variables|Default|GetCurrentPawn)) (Math|Conversions|ToVector(Float) 2e+03))) (.y (* (Transformation|GetActorForwardVector (Variables|Default|GetCurrentPawn)) (Math|Conversions|ToVector(Float) 2e+03))) 2e+03) true true)
      (Variables|/Script/FortniteGame.FortPawn|SetbIgnoreNextFallingDamage true (Variables|Default|GetCurrentPawn)))
    (:Remote
      (bind _asfort_player_pawn_athena_2 (Utilities|Casting|CastToFortPlayerPawnAthena OtherActor)) ; this checks, on the server, if a player touched this, not a bot or wildlife
      (Audio|PlaySoundatLocation "/Game/Sounds/Fort_Traps/BouncePad_Trap/Bounce_Pad_Pickup_Cue.Bounce_Pad_Pickup_Cue" (Transformation|GetActorLocation _asfort_player_pawn_athena_2)))))

(fn ConstructionScript)
