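; Example manager building actor in a ChrisK test map that detects when all of its linked buttons are pressed.
; On BeginPlay it makes itself undamageable and invulnerable.
; A custom check event runs only with authority, calls a function that loops over the AllButtons array of BGA_ButtonManagerExample actors, and prints a debug message if that returns true.
; The function clears a local all-pressed flag for any unpressed button.
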(event EventBeginPlay
  ; Set Invulnerable
  (Variables|Actor|SetbCanBeDamaged false)
  (Variables|BuildingActor|SetbIsInvulnerable true))

(event Custom|EventCheckAllButtons
  ; Check to see if all buttons are pressed
  (Networking|SwitchHasAuthority
    (:Authority
      (bind _allbuttonsareactive (CallFunction|CheckAllButtonsAreActive))
      (if _allbuttonsareactive
        (Development|PrintString "All Buttons Are Pressed" true true "(R=0.000000,G=0.660000,B=1.000000,A=1.000000)" 1e+01)))))

(fn CheckAllButtonsAreActive
  (bind _array_element (Utilities|Array|ForEachLoopwithBreak (Variables|Default|GetAllButtons))
    (:LoopBody
      (if (not (Variables|/Game/NeverCook/Maps/Testmaps/ChrisK/ActorExamples/BGA_ButtonManagerExample.BGA_ButtonManagerExample_C|GetButtonIsPressed _array_element))
        (LocalVariables|SetLocalAllButtonsPressed false)))
    (:Completed
      _))
  (return (LocalVariables|GetLocalAllButtonsPressed)))

(fn ConstructionScript)
