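; Example invulnerable gameplay actor that plays a Niagara effect and sound when any building prop in its assigned list dies.
; On BeginPlay, the authority binds each prop's OnDied event to PropDied, which passes the dead prop's centroid to PlayEffects.
; PlayEffects spawns the configured system at that point plus an offset and the sound at the centroid, only on its Remote (non-authority) branch.
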(event EventBeginPlay
  ; Set Invulnerable
  (Variables|Actor|SetbCanBeDamaged false)
  (Variables|BuildingActor|SetbIsInvulnerable true)
  (Networking|SwitchHasAuthority
    (:Authority
      (CallFunction|BindtoPropsinList))))

(event Custom|BindToPropsInList
  ; Bind Props
  (Networking|SwitchHasAuthority
    (:Authority
      (for _array_element (Variables|Default|GetActorsToPlayEffectsOnDestroy)
        (bind _asbuilding_prop (Utilities|Casting|CastToBuildingProp _array_element))
        (EventDispatchers|BindEventtoOnDied _asbuilding_prop (delegate Custom|PropDied))))))

(event Custom|PropDied
  (params (DamagedActor :type Actor%20Object%20Reference) (Damage :type Float%20%28single-precision%29) (InstigatedBy :type Controller%20Object%20Reference) (DamageCauser :type Actor%20Object%20Reference) (HitLocation :type Vector) (FHitComponent :type Primitive%20Component%20Object%20Reference) (BoneName :type Name) (Momentum :type Vector))
  ; Bind Props
  (bind _asbuilding_prop (Utilities|Casting|CastToBuildingProp DamagedActor))
  (CallFunction|PlayEffects (Building|GetCentroid _asbuilding_prop)))

(event Custom|PlayEffects
  (params (Location :type Vector))
  ; Play FX on died
  (Networking|SwitchHasAuthority
    (:Remote
      (Niagara|SpawnSystematLocation (Variables|Default|GetParticleSystem) (+ Location (Variables|Default|GetEffectsOffset)))
      (Audio|SpawnSoundatLocation (Variables|Default|GetSoundCue) Location))))

(fn ConstructionScript)
