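; Test-map example building actor that spawns and tosses loot when a player pawn overlaps its collision box.
; The actor is invulnerable, and on authority it spawns pickups from a configurable loot tier at its billboard.
; Each pickup is tossed along the billboard's forward vector at a set velocity plus scaled random spread.
; Remote machines play pirate chest sound and Niagara effects; no one-shot guard limits retriggering.
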(event EventBeginPlay
  ; Make Actor Invulnerable
  (Variables|BuildingActor|SetbIsInvulnerable true)
  (Variables|Actor|SetbCanBeDamaged false))

(event OnComponentBeginOverlap(CollisionBox)
  (params (OverlappedComponent :type Primitive%20Component%20Object%20Reference) (OtherActor :type Actor%20Object%20Reference) (OtherComp :type Primitive%20Component%20Object%20Reference) (OtherBodyIndex :type Integer) (bFromSweep :type Boolean) (SweepResult :type Hit%20Result%20Structure%20%28by%20ref%29))
  (bind _billboard (Variables|Default|GetBillboard))
  (bind _returnvalue (Transformation|GetForwardVector _billboard))
  (bind _toss_velocity (Variables|Default|GetToss%20Velocity))
  (bind _returnvalue_2 (Math|Conversions|ToVector(Float) _toss_velocity))
  (bind _returnvalue_3 (* _returnvalue _returnvalue_2))
  ; Trigger Events
  (Networking|SwitchHasAuthority
    (:Authority
      (bind _asfort_player_pawn_athena (Utilities|Casting|CastToFortPlayerPawnAthena OtherActor)) ; this checks, on the server, if a player touched this, not a bot or wildlife
      (Variables|Default|SetCurrentPawn _asfort_player_pawn_athena)
      (Variables|Default|SetImpulse _returnvalue_3)
      (bind _returnvalue_4 (Inventory|SpawnPickupInWorldWithLootTier (Variables|Default|GetLoot%20Tier%20to%20Fire) (Transformation|GetWorldLocation _billboard) 0 0 true true _returnvalue_3))
      (Variables|Default|SetLoot _returnvalue_4)
      (for _array_element (Variables|Default|GetLoot)
        (bind _asfort_pickup_athena (Utilities|Casting|CastToFortPickupAthena _array_element))
        (Pickup|TossPickupinDirection _asfort_pickup_athena (+ (Math|Vector|MakeVector (* (Variables|Default|GetTossSpreadMultiplier) (Math|Random|RandomFloatinRange -4e+02 4e+02)) (* (Math|Random|RandomFloatinRange -4e+02 4e+02) (Variables|Default|GetTossSpreadMultiplier)) (* (Math|Random|RandomFloatinRange 1e+02 2.5e+02) (Variables|Default|GetTossSpreadMultiplier))) (* (Transformation|GetForwardVector (Variables|Default|GetBillboard)) (Variables|Default|GetToss%20Velocity))) self false false)))
    (:Remote
      (bind _asfort_player_pawn_athena_2 (Utilities|Casting|CastToFortPlayerPawnAthena OtherActor)) ; this checks, on the server, if a player touched this, not a bot or wildlife
      (Audio|PlaySoundatLocation "/Game/Sounds/Foley_Loot/Containers/Pirate_Booty/Pirate_Booty_Chest_Open_Cue.Pirate_Booty_Chest_Open_Cue" (Transformation|GetActorLocation _asfort_player_pawn_athena_2))
      (Niagara|SpawnSystematLocation "/Game/Effects/Fort_Effects/Gameplay/PirateChests/NS_Loot_Chest.NS_Loot_Chest" (Transformation|GetWorldLocation (Variables|Default|GetBillboard))))))

(fn ConstructionScript)
