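; Example Fortnite trigger actor that reacts to player pawns entering and leaving a collision box, and is made invulnerable on BeginPlay.
; The server sets IsInsideTrigger on the first player's entry and clears it only when no player pawns still overlap.
; The variable's OnRep turns a point light green or red.
; Custom events, called on those transitions, play enter or exit audio only on non-authority machines.
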(event EventBeginPlay
  ; Make Actor Invulnerable
  (Variables|BuildingActor|SetbIsInvulnerable true)
  (Variables|Actor|SetbCanBeDamaged false))

(event OnComponentBeginOverlap(CollisionBox)
  (params (OverlappedComponent :type Primitive%20Component%20Object%20Reference) (OtherActor :type Actor%20Object%20Reference) (OtherComp :type Primitive%20Component%20Object%20Reference) (OtherBodyIndex :type Integer) (bFromSweep :type Boolean) (SweepResult :type Hit%20Result%20Structure%20%28by%20ref%29))
  ; Trigger Events
  (Networking|SwitchHasAuthority
    (:Authority
      (Utilities|Casting|CastToFortPlayerPawnAthena OtherActor) ; this checks, on the server, if a player touched this, not a bot or wildlife
      (if (not (Variables|Default|GetIsInsideTrigger))
        (Variables|Default|SetIsInsideTrigger true)
        (CallFunction|PlayEnterAudio)))))

(event OnComponentEndOverlap(CollisionBox)
  (params (OverlappedComponent :type Primitive%20Component%20Object%20Reference) (OtherActor :type Actor%20Object%20Reference) (OtherComp :type Primitive%20Component%20Object%20Reference) (OtherBodyIndex :type Integer))
  ; Trigger Events
  (Networking|SwitchHasAuthority
    (:Authority
      (Utilities|Casting|CastToFortPlayerPawnAthena OtherActor) ; this checks, on the server, if a player touched this, not a bot or wildlife
      (if (== (Utilities|Array|Length (Collision|GetOverlappingActors (Variables|Default|GetCollisionBox) "/Script/FortniteGame.FortPlayerPawnAthena" :!class /Script/Engine.PrimitiveComponent)) 0)
        (Variables|Default|SetIsInsideTrigger false)
        (CallFunction|PlayExitAudio)))))

(event Custom|PlayEnterAudio
  ; Client Audio on enter and exit
  (Networking|SwitchHasAuthority
    (:Remote
      (Audio|Components|Audio|Play (Variables|Default|GetEnterAudio) 0.0))))

(event Custom|PlayExitAudio
  ; Client Audio on enter and exit
  (Networking|SwitchHasAuthority
    (:Remote
      (Audio|Components|Audio|Play (Variables|Default|GetExitAudio) 0.0))))

(fn OnRep_IsInsideTrigger
  (if (Variables|Default|GetIsInsideTrigger)
    (Rendering|Components|Light|SetLightColor (Variables|Default|GetPointLight) "(R=0.000000,G=1.000000,B=0.000000,A=1.000000)" true)
    (else
      (Rendering|Components|Light|SetLightColor (Variables|Default|GetPointLight) "(R=1.000000,G=0.000000,B=0.000000,A=1.000000)" true))))

(fn ConstructionScript)
