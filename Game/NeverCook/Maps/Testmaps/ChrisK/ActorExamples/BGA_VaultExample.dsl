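; Example vault actor in a test map whose door slides upward once when the custom OpenVault event fires.
; BeginPlay makes the actor invulnerable and caches the door mesh's relative start position plus a raised end position.
; OpenVault passes a DoOnce into a timeline that lerps the mesh between those positions.
; A timeline event track plays the bunker door sound only on the non-authority side.
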(event EventBeginPlay
  (bind _doormesh (Variables|Default|GetDoorMesh))
  (bind _relativelocation (Variables|/Script/Engine.SceneComponent|GetRelativeLocation _doormesh))
  ; Init
  (Variables|BuildingActor|SetbIsInvulnerable true)
  (Fortnite|SetCanbeDamaged self)
  (Variables|Default|SetDoorStartLocation _relativelocation)
  (Variables|Default|SetDoorEndLocation (+ _relativelocation "0.000000,0.000000,300.000000")))

(event Custom|OpenVault
  ; Open Vault Door
  (Utilities|FlowControl|DoOnce false
    (:Completed
      (bind (_direction _lerp) (Timeline|OpenVaultDoor Play)
        (:Update
          (Transformation|SetRelativeLocation (Variables|Default|GetDoorMesh) (Math|Vector|Lerp(Vector) (Variables|Default|GetDoorStartLocation) (Variables|Default|GetDoorEndLocation) _lerp) false false))
        (:DoorOpening
          (Networking|SwitchHasAuthority
            (:Remote
              (Audio|PlaySoundatLocation "/Game/Sounds/Fort_Doors/World/Door_Bunker_Open_Cue.Door_Bunker_Open_Cue"))))))))

(fn ConstructionScript)
