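; Test-map example of a damageable building prop cube that changes mesh and material when damaged.
; On server damage below half health, a DoOnce sets CubeIsDamaged and calls the cosmetic event with the hit location.
; CubeIsDamaged's OnRep handles the damaged-mesh and red-material swap, which requires a Movable static mesh component.
; The cosmetic event plays spark audio and short-lived Niagara damage FX only on the non-authority branch.
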(event EventOnDamageServer
  (params (Damage :type Float%20%28single-precision%29) (DamageTags :type Gameplay%20Tag%20Container%20Structure%20%28by%20ref%29) (Momentum :type Vector :default "0, 0, 0") (HitInfo :type Hit%20Result%20Structure%20%28by%20ref%29) (InstigatedBy :type Controller%20Object%20Reference) (DamageCauser :type Actor%20Object%20Reference) (EffectContext :type Gameplay%20Effect%20Context%20Handle%20Structure))
  (bind (_bblockinghit _binitialoverlap _time _distance _location) (Collision|BreakHitResult HitInfo))
  ; Damage Logic
  (if (< (Health|GetHealthPercent self) 0.5)
    (Utilities|FlowControl|DoOnce false
      (:Completed
        (Variables|Default|SetCubeIsDamaged true)
        (CallFunction|PlayDamagedAudioandFX _location)))))

(event Custom|PlayDamagedAudioAndFX
  (params (HitVector :type Vector))
  ; Client Cosmetics
  (Networking|SwitchHasAuthority
    (:Remote
      (Audio|PlaySoundatLocation "/Game/Sounds/Fort_Traps/Transformer/ElectricitySpark_Zap_SoundCue.ElectricitySpark_Zap_SoundCue" HitVector)
      (bind _returnvalue (Niagara|SpawnSystematLocation "/Game/Athena/DrivableVehicles/FX/NS_Vehicle_Damage.NS_Vehicle_Damage" HitVector))
      (Utilities|FlowControl|Delay 2.0)
      (Components|Activation|Deactivate _returnvalue))))

(fn OnRep_CubeIsDamaged
  (bind _staticmeshcomponent (Variables|Debug|GetStaticMeshComponent))
  (if (Variables|Default|GetCubeIsDamaged)
    (Components|StaticMesh|SetStaticMesh _staticmeshcomponent "/Game/Athena/Asteria/Maps/Test/Dev/PStone/TestMeshes/cube_2x2x2_damaged.cube_2x2x2_damaged") ; to change a mesh in a buildingprop actor, you must set the "Mobility" of the Static Mesh Component to "Moveable" rather than Static
    (Rendering|Material|SetMaterial _staticmeshcomponent 0 "/Game/Environments/Asteria/Materials/Proto/M_Asteria_Proto_2Sided_Red.M_Asteria_Proto_2Sided_Red")))

(fn ConstructionScript)
